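# Material function for the Lima Bean backpack that draws a cel-shaded, inked look through emissive, with black base color and specular.
# Sun or camera-space light vectors drive shadow, lit and highlight bands for body and goggles colors, apparently blended by vertex green.
# Time-of-day scaling and an optional LUT adjust color, darkened by camera-distance-scaled distance-field ink lines and an optional Fresnel outline.
# Vertex red selects roughness.
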MaterialFunction /BRCosmetics/Accessories/FORT_Backpacks/Backpack_M_MED_Lima_Bean/Materials/MF_Lima_Bean
graph {nodes [
  n1 = Constant3Vector()
  n2 = VertexColor()
  n3 = SphereMask(A=@n2.R, B=Constant() {r=0.5}) {attenuationRadius=0.1}
  n4 = Constant()
  n5 = SetMaterialAttributes!(
   MaterialAttributes=FunctionInput() {inputType=FunctionInput_MaterialAttributes}
   BaseColor=@n1
   EmissiveColor=NamedRerouteUsage()
   Specular=@n1
   Roughness=LinearInterpolate(A=LinearInterpolate(Alpha=@n3) {constA=0.5, constB=0}, Alpha=SphereMask(A=@n2.R, B=@n4) {attenuationRadius=0.1}) {constA=0.5}
  ) {attributeSetTypes [
    #1
    #2
    #3
    #4
   ]}
  FunctionOutput(None=@n5)
  n6 = DotProduct(A=NamedRerouteUsage(), B=PixelNormalWS())
  n7 = ScalarParameter() {defaultValue=-0.2, parameterName=TerminatorOffset}
  NamedRerouteDeclaration(None=SmoothStep(Max=ScalarParameter() {defaultValue=0.1, group=LimaBean, parameterName=TerminatorSmoothness}, Value=Add(A=@n6, B=@n7) {constB=-0.3}) {constMax=0}) {name=terminator, nodeColor=<rgba 1 0 0.9609 1>}
  n8 = VectorParameter() {group=LimaBean, parameterName=CustomColor, defaultValue=<rgba 0.5647 0.0056 0.0056 1>}
  n9 = VectorParameter() {group=LimaBean, parameterName=CustomColorLit, defaultValue=<rgba 0.5647 0.0336 0.0056 1>}
  n10 = NamedRerouteUsage()
  NamedRerouteDeclaration(None=LinearInterpolate(
    A=VectorParameter() {group=LimaBean, parameterName=CustomColorShadow, defaultValue=<rgba 0.2346 0 0.0395 1>}.RGB
    B=LinearInterpolate(A=@n8.RGB, B=@n9.RGB, Alpha=@n10)
    Alpha=NamedRerouteUsage()
   )) {name=bodyColor, nodeColor=<rgba 0.6172 0 1 1>}
  n11 = VectorParameter() {group=LimaBean, parameterName=maskShadow, defaultValue=<rgba 0.0529 0.2086 0.3712 1>}
  n12 = VectorParameter() {group=LimaBean, parameterName=maskColor, defaultValue=<rgba 0.2384 0.5972 0.7379 1>}
  n13 = NamedRerouteUsage()
  NamedRerouteDeclaration(None=LinearInterpolate(
    A=LinearInterpolate(A=@n11.RGB, B=@n12.RGB, Alpha=@n13)
    B=VectorParameter() {group=LimaBean, parameterName=maskHilight, defaultValue=<rgba 0.9734 0.9734 0.9734 1>}.RGB
    Alpha=NamedRerouteUsage()
   )) {name=GogglesColor, nodeColor=<rgba 0.7578 0 1 1>}
  n14 = NamedRerouteUsage()
  n15 = PixelNormalWS()
  NamedRerouteDeclaration(None=SphereMask(A=DotProduct(A=@n14, B=@n15), B=Constant() {r=1}) {attenuationRadius=0.06, hardnessPercent=80}) {name=Highlight, nodeColor=<rgba 0.6875 1 0 1>}
  n16 = TextureSampleParameter2D() {
   group=LimaBeanDF
   parameterName=DistanceField_InkLines
   samplerType=SAMPLERTYPE_Grayscale
   texture=/Game/Global/Textures/Default/Blanks/T_BlackLinearGrayscale_sRGB
  }
  n17 = TextureSampleParameter2D() {
   group=LimaBeanDF
   parameterName=DistanceField_InkLines_Mobile
   samplerType=SAMPLERTYPE_Grayscale
   texture=&T_BlackLinearGrayscale_sRGB
  }
  n18 = ChannelMaskParameter(None=ShadingPathSwitch(Default=@n16.RGB, Deferred=@n16.RGB, Mobile=@n17.RGB) {desc=Migrated from FeatureLevelSwitch}) {group=LimaBeanDF, parameterName=DistanceField_Channel}
  n19 = ScalarParameter() {defaultValue=9, group=LimaBeanDF, parameterName=DF_BlurPow}
  n20 = OneMinus(None=Multiply(A=Power(Base=@n18.R, Exp=@n19) {constExponent=3}, B=ScalarParameter() {defaultValue=1, group=LimaBeanDF, parameterName=DF_BlurMult}) {constB=0.5})
  n21 = Constant() {r=1}
  n22 = ScalarParameter() {defaultValue=0.05, group=LimaBeanDF, parameterName=DF_LineThickness_Near}
  n23 = ScalarParameter() {defaultValue=0.15, group=LimaBeanDF, parameterName=DF_LineThickness_Far}
  n24 = ScalarParameter() {defaultValue=100, group=LimaBeanDF, parameterName=DF_FarDistance}
  n25 = Subtract(A=Distance(A=WorldPosition().XYZ, B=CameraPositionWS()), B=@n24)
  n26 = Subtract(A=ScalarParameter() {defaultValue=500, group=LimaBeanDF, parameterName=DF_NearDistance}, B=@n24)
  n27 = LinearInterpolate(
   A=ShadingPathSwitch(Default=@n22, Deferred=@n22, Mobile=ScalarParameter() {defaultValue=0.05, group=LimaBeanDF, parameterName=DF_LineThickness_Near_Mobile}) {desc=Migrated from FeatureLevelSwitch}
   B=ShadingPathSwitch(Default=@n23, Deferred=@n23, Mobile=ScalarParameter() {defaultValue=0.15, group=LimaBeanDF, parameterName=DF_LineThickness_Far_Mobile}) {desc=Migrated from FeatureLevelSwitch}
   Alpha=Saturate(None=Divide(A=@n25, B=@n26))
  )
  n28 = ScalarParameter() {defaultValue=0.8, group=LimaBeanDF, parameterName=DF_LineHardness}
  n29 = OneMinus(None=Saturate(None=SphereMask(A=@n18.R, B=@n21, Radius=@n27, Hardness=@n28) {attenuationRadius=0.1}))
  NamedRerouteDeclaration(None=Saturate(None=Multiply(A=@n20, B=@n29))) {name=DF Lines, nodeColor=<rgba 0 1 0.0625 1>}
  NamedRerouteDeclaration(None=SmoothStep(Max=ScalarParameter() {defaultValue=2, group=LimaBean, parameterName=LitSmoothness}, Value=@n6) {constMax=0}) {name=shading, nodeColor=<rgba 0.5 1 0 1>}
  n30 = Multiply(
   A=LinearInterpolate(A=NamedRerouteUsage(), B=NamedRerouteUsage(), Alpha=VertexColor().G)
   B=MaterialFunctionCall(
    Value_Morning=ScalarParameter() {defaultValue=1, group=LimaBean, parameterName=lima_morning}
    Value_Day=ScalarParameter() {defaultValue=1, group=LimaBean, parameterName=lima_day}
    Value_Evening=ScalarParameter() {defaultValue=1, group=LimaBean, parameterName=lima_evening}
    Value_Night=ScalarParameter() {defaultValue=1, group=LimaBean, parameterName=lima_night}
   ) {materialFunction=&Scalar_Time_of_Day}.Result
  )
  n31 = LinearInterpolate(
   A=@n30
   B=MaterialFunctionCall!(In=@n30, LUT, LUT Multiplier) {materialFunction=/Game/Packages/Fortress_SharedMaterials/Anime_Master_Material/MF_ApplyLUT}.Result
   Alpha=ScalarParameter() {defaultValue=1, group=LimaBean, parameterName=useLut}
  ) {constAlpha=1}
  n32 = ScalarParameter() {defaultValue=1, group=LimaBean, parameterName=baseColorMult}
  n33 = NamedRerouteUsage()
  n34 = Multiply(A=@n33, B=NamedRerouteUsage())
  NamedRerouteDeclaration(None=LinearInterpolate(
    A=Constant3Vector()
    B=Multiply(A=@n31, B=@n32) {constB=0.9}
    Alpha=StaticSwitchParameter(True=@n34, False=@n33) {group=LimaBean, parameterName=useFresnelOutline}
   ) {constAlpha=1}) {name=Color, nodeColor=<rgba 0.5 1 0 1>}
  n35 = DotProduct(A=CameraVectorWS(), B=VectorParameter() {defaultValue=<rgba 0 0 1 0>}.RGB)
  n36 = LinearInterpolate(Alpha=Power(Base=Abs(None=@n35))) {constA=0.55, constB=0.7}
  n37 = InverseLinearInterpolate(A=@n36, B=Add(A=@n36) {constB=0.02}, Value=Fresnel(Normal=PixelNormalWS()) {baseReflectFraction=0, exponent=1}) {constA=0.8, constB=0.82}
  NamedRerouteDeclaration(None=OneMinus(None=Saturate(None=@n37))) {name=FresnelOutline, nodeColor=<rgba 1 0 0.5391 1>}
  n38 = CollectionParameter() {
   collection=&FortniteMaterialParameters
   parameterName=SunlightDirectionalVector
  }
  n39 = ComponentMask(None=@n38) {b=true, g=true, r=true}
  n40 = VectorParameter() {group=LimaBean, parameterName=lima_body_rimLightVector, sortPriority=100, defaultValue=<rgba 1.5 1 -1 0>}
  n41 = ComponentMask(None=Transform(None=@n40.RGB) {transformSourceType=TRANSFORMSOURCE_Camera}) {b=true, g=true, r=true}
  NamedRerouteDeclaration(None=StaticSwitchParameter(True=@n39, False=Normalize(VectorInput=@n41)) {defaultValue=true, group=LimaBean, parameterName=useSunVector}) {name=Body Light Vector, nodeColor=<rgba 1 0.1641 0 1>}
  n42 = VectorParameter() {group=LimaBean, parameterName=lima_goggles_rimLightVector, sortPriority=101, defaultValue=<rgba 0.3 0.5 -1 0>}
  n43 = ComponentMask(None=Transform(None=@n42.RGB) {transformSourceType=TRANSFORMSOURCE_Camera}) {b=true, g=true, r=true}
  NamedRerouteDeclaration(None=StaticSwitchParameter(True=@n39, False=Normalize(VectorInput=@n43)) {defaultValue=true, group=LimaBean, parameterName=useSunVector}) {name=GogglesLightVector, nodeColor=<rgba 1 0.9375 0 1>}
  Normalize(VectorInput=Add(A=CameraVectorWS(), B=NamedRerouteUsage()))
  ScalarParameter() {defaultValue=0.15, parameterName=lineThickness}
 ]}
