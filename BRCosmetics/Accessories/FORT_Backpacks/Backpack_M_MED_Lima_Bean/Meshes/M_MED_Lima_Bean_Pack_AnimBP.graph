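# Animation Blueprint for the Lima Bean backpack cosmetic, animating it with the wearer's movement and DBNO state.
# An Idle/Run/Stop/DBNO state machine, with a crouch-run variant, feeds a weighted spine-pelvis attachment constraint, two spring AnimDynamics nodes and a pack upscale.
# Top and bottom constraint bones then follow the backpack, or a weighted spine, pelvis and self blend during DBNO.
# The update event sets those constraint alphas and dampens dynamics while skydive floating.
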AnimBlueprint /BRCosmetics/Accessories/FORT_Backpacks/Backpack_M_MED_Lima_Bean/Meshes/M_MED_Lima_Bean_Pack_AnimBP
target_skeleton=&/M_MED_Lima_Bean_Pack_Skeleton
graphs [
 AnimGraph {nodes [
   Input Pose()
   AnimGraphNode_StateMachine_0 = New State Machine() {is=title AnimGraphNode_StateMachine}
   n1 = "Constraint - dyn_backpack_attachment_jnt (PPPP)"!(
    ComponentPose=Local To Component(LocalPose=@AnimGraphNode_StateMachine_0)
    Alpha=1
    ConstraintWeights_0=0.4527
    ConstraintWeights_1:float=0.4069
    ConstraintWeights_2:float=0.092
    ConstraintWeights_3:float=0.0484
   ) -> (Pose) {
    alphaBoolBlend {blendOption=Linear}
    boneToModify {boneName=dyn_backpack_attachment_jnt}
    constraintSetup [
     {targetBone {boneName=dyn_spine2}}
     {targetBone {boneName=dyn_spine1}}
     {targetBone {boneName=dyn_spine3}}
     {targetBone {boneName=dyn_pelvis_down}}
    ]
    constraintWeights [1,1,1,1]
   }
   n2 = Get skydiveDampen() -> (skydiveDampen:double)
   n3 = Anim Dynamics - Bone: dyn_backpack(ComponentPose=@n1, Alpha=@n2, ExternalForce="0,0,0") {
    bAngularSpring=true
    bLinearSpring=true
    bOverrideLinearDamping=true
    bUseGravityOverride=true
    gravityScale=0.35
    linearDampingOverride=0.999
    linearSpringConstant=90
    alphaBoolBlend {blendOption=Linear}
    alphaScaleBiasClamp {scale=0.73}
    boundBone {boneName=dyn_backpack}
    componentLinearAccScale=<0.3 0.3 0.3>
    componentLinearVelScale=<0.3 0.3 0.3>
    gravityOverride {z=-600}
    physicsBodyDefinitions [{
      sphereCollisionRadius=0
      boundBone {boneName=dyn_backpack}
      boxExtents=<90 100 100>
      constraintSetup {angularLimitsMax=<0 0 0>, angularLimitsMin=<0 0 0>, angularTarget=<1 0 0>, linearAxesMax=<2 0 0>}
      localJointOffset=<1 0 0>
     }]
    relativeSpaceBone {boneName=root}
   }
   n4 = Anim Dynamics - Bone: pack(ComponentPose=@n3, Alpha=@n2, ExternalForce="0,0,0") {
    bLinearSpring=true
    bOverrideLinearDamping=true
    bUseGravityOverride=true
    gravityScale=0.35
    linearDampingOverride=0.999
    linearSpringConstant=90
    alphaBoolBlend {blendOption=Linear}
    boundBone {boneName=pack}
    componentLinearAccScale=<0.25 0.25 0.25>
    componentLinearVelScale=<0.25 0.25 0.25>
    gravityOverride {z=-500}
    physicsBodyDefinitions [{
      sphereCollisionRadius=0
      boundBone {boneName=pack}
      boxExtents=<30 40 40>
      constraintSetup {angularLimitsMax=<0 0 0>, angularLimitsMin=<0 0 0>, angularTarget=<1 0 0>, linearAxesMax=<0 0 2>, linearAxesMin=<0 0 -1>}
      localJointOffset=<1 0 0>
     }]
    relativeSpaceBone {boneName=root}
   }
   n5 = "Transform (Modify) Bone - Bone: pack"(ComponentPose=@n4, Translation="0,0,0", Rotation="0,0,0", Scale="1.15,1.15,1.15") {scaleMode=BMM_Replace, scaleSpace=BCS_BoneSpace, alphaBoolBlend {blendOption=Linear}, boneToModify {boneName=pack}, scale=<1.15 1.15 1.15>}
   n6 = Get isDBNOoff() -> (isDBNOoff:double)
   n7 = "Constraint - dyn_backpack_constraintTop (P)"(ComponentPose=@n5, Alpha=@n6, ConstraintWeights_0=1) {
    alphaBoolBlend {blendOption=Linear}
    boneToModify {boneName=dyn_backpack_constraintTop}
    constraintSetup [{targetBone {boneName=dyn_backpack}}]
    constraintWeights [1]
   }
   n8 = "Constraint - dyn_backpack_constraintBottom (P)"(ComponentPose=@n7, Alpha=@n6, ConstraintWeights_0=1) {
    alphaBoolBlend {blendOption=Linear}
    boneToModify {boneName=dyn_backpack_constraintBottom}
    constraintSetup [{targetBone {boneName=dyn_backpack}}]
    constraintWeights [1]
   }
   n9 = Get isDBNO() -> (isDBNO:double)
   n10 = "Constraint - dyn_backpack_constraintTop (PPPPPP)"!(
    ComponentPose=@n8
    Alpha=@n9
    ConstraintWeights_0=0.0002
    ConstraintWeights_1:float=0.0514
    ConstraintWeights_2:float=0.3515
    ConstraintWeights_3:float=0.0969
    ConstraintWeights_4:float=0.4754
    ConstraintWeights_5:float=0.0246
   ) -> (Pose) {
    alphaBoolBlend {blendOption=Linear}
    boneToModify {boneName=dyn_backpack_constraintTop}
    constraintSetup [
     {targetBone {boneName=dyn_pelvis_down}}
     {targetBone {boneName=dyn_spine1}}
     {targetBone {boneName=dyn_spine2}}
     {targetBone {boneName=dyn_spine3}}
     {targetBone {boneName=dyn_backpack_constraintTop}}
     {targetBone {boneName=dyn_backpack_constraintBottom}}
    ]
    constraintWeights [1,1,1,1,1,1]
   }
   n11 = "Constraint - dyn_backpack_constraintBottom (PPPPPP)"!(
    ComponentPose=@n10
    Alpha=@n9
    ConstraintWeights_0=0.2752
    ConstraintWeights_1:float=0.1534
    ConstraintWeights_2:float=0.0261
    ConstraintWeights_3:float=0.0453
    ConstraintWeights_4:float=0.0193
    ConstraintWeights_5:float=0.4807
   ) -> (Pose) {
    alphaBoolBlend {blendOption=Linear}
    boneToModify {boneName=dyn_backpack_constraintBottom}
    constraintSetup [
     {targetBone {boneName=dyn_pelvis_down}}
     {targetBone {boneName=dyn_spine1}}
     {targetBone {boneName=dyn_spine2}}
     {targetBone {boneName=dyn_spine3}}
     {targetBone {boneName=dyn_backpack_constraintTop}}
     {targetBone {boneName=dyn_backpack_constraintBottom}}
    ]
    constraintWeights [1,1,1,1,1,1]
   }
   Output Pose(Result=Component To Local(ComponentPose=@n11)) {layerGroup=None}
  ]}
 New State Machine {
  entry=Idle
  path=AnimGraph.AnimGraphNode_StateMachine_0.New State Machine
  ids=id
  nodes [
   AnimStateNode_0 = Idle()
   AnimStateNode_1 = Run() {is=title AnimStateNode}
   AnimStateNode_2 = DBNO() {is=title AnimStateNode}
   AnimStateNode_3 = Stop() {is=title AnimStateNode}
   AnimStateTransitionNode_0 = AnimStateTransitionNode() {crossfadeDuration=0}
   AnimStateTransitionNode_10 = AnimStateTransitionNode()
   AnimStateTransitionNode_11 = AnimStateTransitionNode() {crossfadeDuration=0}
   AnimStateTransitionNode_3 = AnimStateTransitionNode()
   AnimStateTransitionNode_4 = AnimStateTransitionNode() {crossfadeDuration=0.1}
   AnimStateTransitionNode_7 = AnimStateTransitionNode() {crossfadeDuration=0}
   AnimStateTransitionNode_9 = AnimStateTransitionNode() {bAutomaticRuleBasedOnSequencePlayerInState=true, crossfadeDuration=0}
  ]
  links [
   {
    dst=Idle
    rule=AnimGraph.AnimGraphNode_StateMachine_0.New State Machine.AnimStateTransitionNode_3.Transition
    src=DBNO
    via=AnimStateTransitionNode_3
   }
   {
    dst=DBNO
    rule=AnimGraph.AnimGraphNode_StateMachine_0.New State Machine.AnimStateTransitionNode_10.Transition
    src=Idle
    via=AnimStateTransitionNode_10
   }
   {
    dst=Run
    rule=AnimGraph.AnimGraphNode_StateMachine_0.New State Machine.AnimStateTransitionNode_0.Transition
    src=Idle
    via=AnimStateTransitionNode_0
   }
   {
    dst=Stop
    rule=AnimGraph.AnimGraphNode_StateMachine_0.New State Machine.AnimStateTransitionNode_4.Transition
    src=Run
    via=AnimStateTransitionNode_4
   }
   {
    dst=Stop
    rule=AnimGraph.AnimGraphNode_StateMachine_0.New State Machine.AnimStateTransitionNode_7.Transition
    src=Run
    via=AnimStateTransitionNode_7
   }
   {
    dst=Idle
    rule=AnimGraph.AnimGraphNode_StateMachine_0.New State Machine.AnimStateTransitionNode_11.Transition
    src=Stop
    via=AnimStateTransitionNode_11
   }
   {
    dst=Idle
    rule=AnimGraph.AnimGraphNode_StateMachine_0.New State Machine.AnimStateTransitionNode_9.Transition
    src=Stop
    via=AnimStateTransitionNode_9
   }
  ]
 }
 Idle {
  path=AnimGraph.AnimGraphNode_StateMachine_0.New State Machine.AnimStateNode_0.Idle
  nodes [Output Animation Pose(Result=Sequence Player 'Idle'() {sequence=/BRCosmetics/Accessories/FORT_Backpacks/Backpack_M_MED_Lima_Bean/Animation/Idle})]
 }
 Run {
  path=AnimGraph.AnimGraphNode_StateMachine_0.New State Machine.AnimStateNode_1.Run
  nodes [
   n1 = Sequence Player 'Crouch_Run'() {groupName=Run, method=SyncGroup, sequence=/BRCosmetics/Accessories/FORT_Backpacks/Backpack_M_MED_Lima_Bean/Animation/Crouch_Run}
   Output Animation Pose(Result=Blend Poses by bool(
     BlendPose_0=@n1
     BlendPose_1=Sequence Player 'Run'() {groupName=Run, method=SyncGroup, sequence=/BRCosmetics/Accessories/FORT_Backpacks/Backpack_M_MED_Lima_Bean/Animation/Run}
     BlendTime_0=0
     BlendTime_1=0
     bActiveValue=Get bIsCrouching()
    ) {blendType=Linear})
  ]
 }
 DBNO {
  path=AnimGraph.AnimGraphNode_StateMachine_0.New State Machine.AnimStateNode_2.DBNO
  nodes [Output Animation Pose(Result=Sequence Player 'DBNO'!(bLoopAnimation:bool=False) -> (Pose) {bLoopAnimation=false, sequence=/BRCosmetics/Accessories/FORT_Backpacks/Backpack_M_MED_Lima_Bean/Animation/DBNO})]
 }
 Stop {
  path=AnimGraph.AnimGraphNode_StateMachine_0.New State Machine.AnimStateNode_3.Stop
  nodes [Output Animation Pose(Result=Sequence Player 'Stop'!(bLoopAnimation:bool=False) -> (Pose) {bLoopAnimation=false, sequence=/BRCosmetics/Accessories/FORT_Backpacks/Backpack_M_MED_Lima_Bean/Animation/Stop})]
 }
 Transition {
  path=AnimGraph.AnimGraphNode_StateMachine_0.New State Machine.AnimStateTransitionNode_0.Transition
  nodes [Result(bCanEnterTransition=Get bIsMoving2D())]
 }
 Transition {
  path=AnimGraph.AnimGraphNode_StateMachine_0.New State Machine.AnimStateTransitionNode_10.Transition
  nodes [Result(bCanEnterTransition=Get bIsDBNO())]
 }
 Transition {
  path=AnimGraph.AnimGraphNode_StateMachine_0.New State Machine.AnimStateTransitionNode_11.Transition
  nodes [Result(bCanEnterTransition=Get bIsDBNO())]
 }
 Transition {
  path=AnimGraph.AnimGraphNode_StateMachine_0.New State Machine.AnimStateTransitionNode_3.Transition
  nodes [Result(bCanEnterTransition=NOT Boolean(A=Get bIsDBNO()))]
 }
 Transition {
  path=AnimGraph.AnimGraphNode_StateMachine_0.New State Machine.AnimStateTransitionNode_4.Transition
  nodes [Result(bCanEnterTransition=NOT Boolean(A=Get bIsMoving2D()))]
 }
 Transition {
  path=AnimGraph.AnimGraphNode_StateMachine_0.New State Machine.AnimStateTransitionNode_7.Transition
  nodes [Result(bCanEnterTransition=Get bIsDBNO())]
 }
 Transition {path=AnimGraph.AnimGraphNode_StateMachine_0.New State Machine.AnimStateTransitionNode_9.Transition, nodes [Result()]}
 EventGraph {nodes [
   Event BlueprintUpdateAnimation()
   > n1 = Sequence()
   > n2 = Branch(Condition=Get bIsSkydiveFloating())
   > Set skydiveDampen(skydiveDampen:double=0.4) -> (Output_Get:double)
   Set skydiveDampen(execute=@n2.else, skydiveDampen=1)
   n3 = Branch(execute=@n1.then_1, Condition=Get bIsDBNO())
   > Set isDBNO(isDBNO:double=1) -> (Output_Get:double)
   > Set isDBNOoff(isDBNOoff:double=0) -> (Output_Get:double)
   Set isDBNOoff(execute=@n3.else, isDBNOoff=1)
   > Set isDBNO(isDBNO=0)
  ]}
]
variables [
 isDBNO {type=number}
 isDBNOoff {type=number}
 PlayerPawn {type=FortPlayerPawn}
 RunFacingRight {type=boolean}
 skydiveDampen {type=number}
 YawDeltaPositive {type=boolean}
 YawDeltaPositiveLastTick {type=boolean}
]
